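; Spawner for the Plunder job board briefing, which offers the buried treasure job only if enough valid item locations exist.
; On spawn, the authority waits for playlist data, then runs an EQS query and binds to its completion.
; The parent spawn runs only if returned locations cover the scalable chest count; otherwise it prints a debug warning.
; The spawner then gets a three-second lifespan, unbinds the query at EndPlay, and hides its previs mesh in game.
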(event EventBeginPlay
  ;[0 0 0] We want this visible in editor when 'G' is pressed, but not visible in game
  (|CallParentFunction)
  (Development|SetHiddeninGame (Variables|Default|GetPreVisMesh) true true))

(event EventSpawnActor
  (params (New%20Transform :type Transform))
  ;[0 0 0] Before spawning this actor, check if it can find valid locations for the job's item spawning offer before providing an option
  (Networking|SwitchHasAuthority
    (:Authority
      (Class|FortAsyncActionCurrentPlaylistReady|PlaylistDataReadyAsync
        (:OnReady
          (bind _returnvalue (AI|EQS|RunEQSQuery (Variables|Default|GetQuery%20Template) self "AllMatching"))
          (bind _output_get (Variables|Default|SetTempEQS _returnvalue))
          (AI|EQS|BindEventtoOnQueryFinished _output_get (delegate Custom|SpawnerEQSFinished)))))))

(event Custom|SpawnerEQSFinished
  (params (QueryInstance :type EQS%20Query%20Instance%20Object%20Reference) (QueryStatus :type EEnvQueryStatus%20Enum :default "Processing"))
  ;[0 0 0] Before spawning this actor, check if it can find valid locations for the job's item spawning offer before providing an option
  (bind (_resultlocations _returnvalue) (AI|EQS|GetQueryResultsasLocations QueryInstance))
  ;;[0.2016 0.2016 0.2016] Enough locations returned for items requested?
  (if _returnvalue
    (if (>= (Utilities|Array|Length _resultlocations) (Math|Float|Truncate (ScalableFloat|GetValueatLevel (Variables|Default|GetBuriedChestsToSpawn))))
      (|CallParentFunction (stale EventSpawnActor New%20Transform))
      (else
        (Development|PrintString "Not enough valid locations found to spawn buried treasure job for the count requested" true true "(R=1.000000,G=0.000000,B=0.403795,A=1.000000)" 4.0)))
    (else
      (Development|PrintString "No valid locations found to spawn buried treasure job" true true "(R=1.000000,G=0.000000,B=0.403795,A=1.000000)" 4.0)))
  (Actor|SetLifeSpan 3.0))

(event EventEndPlay
  (params (EndPlayReason :type EEndPlayReason%20Enum :default "Destroyed"))
  (bind _tempeqs (Variables|Default|GetTempEQS))
  ;[0 0 0] End play - If a query was set, unbind the event from it
  (|CallParentFunction EndPlayReason)
  (Utilities|IsValid _tempeqs
    (:Is%20Valid
      (AI|EQS|UnbindEventfromOnQueryFinished _tempeqs (EventDispatchers|CreateEvent :!event Custom|SpawnerEQSFinished)))))

(fn ConstructionScript
  (|CallParentFunction))
