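; Spawner Blueprint for the Job Board briefing supply drop, built on a parent spawner class that supplies the spawning logic.
; On BeginPlay it runs the parent logic, then hides its PreVisMesh and that mesh's children in game, leaving it as an editor-only preview.
; The server-spawned and construction hooks only forward to the parent, while overlap and tick are empty.
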(event EventBeginPlay
  ;[0 0 0] We want this visible in editor when 'G' is pressed, but not visible in game
  (|CallParentFunction)
  (Development|SetHiddeninGame (Variables|Default|GetPreVisMesh) true true))

(event Collision|EventActorBeginOverlap
  (params (OtherActor :type Actor%20Object%20Reference)))

(event EventTick
  (params (DeltaSeconds :type Float%20%28single-precision%29)))

(event EventServerActorSpawned
  (params (Actor :type Actor%20Object%20Reference))
  (|CallParentFunction Actor))

(fn ConstructionScript
  (|CallParentFunction))
